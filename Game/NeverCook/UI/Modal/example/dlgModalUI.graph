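# Example editor utility widget for a simple modal dialog, built on EditorUtilityDialogWidget as editor UI.
# A vertical box stacks a descriptive text line, an editable text box (edt) with hint text, and a slider, all evenly padded.
# The text box and slider are exposed as variables, but the event graph holds only unconnected default events, so the asset itself adds no logic.
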EditorUtilityWidgetBlueprint /Game/NeverCook/UI/Modal/example/dlgModalUI
graphs [EventGraph {nodes [
   Event PreConstruct()
   Event Construct()
   Event Tick()
  ]}]
variables []
widgets {
 info {parentClass=/Script/Blutility.EditorUtilityDialogWidget, rootWidgetClass=&VerticalBox, widgetCount=4}
 widgets [
  VerticalBox_0 {widgetClassPath=&VerticalBox, widget=VerticalBox_0}
  TextBlock_33 {
   widgetClassPath=&TextBlock
   parent=VerticalBox_0
   widget=TextBlock_33
   properties {
    text=A simple modal dialog with text and an input field.
    font {
     bForceMonospaced=false
     bMaterialIsStencil=false
     fontMaterial=null
     fontObject=&NotoSans
     letterSpacing=0
     monospacedWidth=1
     size=9
     skewAmount=0
     typefaceFontName=NotoSans-Regular
     outlineSettings {
      bApplyOutlineToDropShadows=false
      bMiteredCorners=false
      bSeparateFillAlpha=false
      outlineMaterial=null
      outlineSize=0
      outlineColor=<rgba 0 0 0 1>
     }
    }
   }
   slot {class=VerticalBoxSlot, properties {padding=<ltrb 10 10 10 10>}}
  }
  edt {
   widgetClassPath=&EditorUtilityEditableTextBox
   bIsVariable=true
   parent=VerticalBox_0
   widget=edt
   properties {hintText=Input text}
   slot {class=VerticalBoxSlot, properties {padding=<ltrb 10 10 10 10>}}
  }
  slider {
   widgetClassPath=/Script/Blutility.EditorUtilitySlider
   bIsVariable=true
   parent=VerticalBox_0
   widget=slider
   slot {class=VerticalBoxSlot, properties {padding=<ltrb 10 10 10 10>}}
  }
 ]
}
