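# Example Editor Utility Widget that demonstrates opening dlgModalUI dialogs in three ways: modal, non-modal and pre-created modal instance.
# The modal button reads the returned result struct directly, while the non-modal one receives it through the OnClosed delegate.
# The instance button copies its text box into the dialog before showing it, then copies the text back regardless of result.
# ShowResults displays the pressed button, dialog text and slider value in a message box.
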EditorUtilityWidgetBlueprint /Game/NeverCook/UI/Modal/example/MainUI
graphs [
 EventGraph {nodes [
   "On Clicked (EditorUtilityButton_0)"() {is=title K2Node_ComponentBoundEvent}
   n1 = Get Dialog UI() -> (Dialog UI:Dlg Modal UI Class Reference)
   > n2 = ShowModalEditorUtilityDialog!(
    execute
    WidgetClass:Editor Utility Dialog Widget Class Reference=@n1
    Title:Text="LOCTEXT(\"Cool Modal Example\")"
    ButtonLayout:EAppMsgType Enum=OkCancel
    Size:Vector 2D Structure
   ) -> (then, ReturnValue:Editor Utility Dialog Result Structure)
   n3 = Break Editor Utility Dialog Result!(EditorUtilityDialogResult:"Editor Utility Dialog Result Structure (by ref)"=@n2) -> (ReturnType:EAppReturnType Enum=~, WidgetInstance:Editor Utility Dialog Widget Object Reference) {is=title K2Node_BreakStruct}
   > ShowResults!(execute, self, Index:EAppReturnType Enum=@n3, Object:Object Reference=@n3.WidgetInstance) -> (then)
   n4 = "On Clicked (EditorUtilityButton)"() {is=title K2Node_ComponentBoundEvent}
   n5 = OnClosed!() -> (OutputDelegate, then, Result:EAppReturnType Enum, WidgetInstance:Editor Utility Dialog Widget Object Reference) {is=title K2Node_CustomEvent}
   ShowEditorUtilityDialog!(
    execute=@n4.then
    WidgetClass:Editor Utility Dialog Widget Class Reference=@n1
    OnClosed:"Delegate (by ref)"=@n5
    Title:Text
    ButtonLayout:EAppMsgType Enum=OkCancel
    Size:Vector 2D Structure
   ) -> (then, ReturnValue:Dlg Modal UI Object Reference)
   "On Clicked (btn_InstanceModal)"() {is=title K2Node_ComponentBoundEvent}
   > n6 = CreateEditorUtilityDialogWidget!(execute, WidgetClass:Editor Utility Dialog Widget Class Reference=@n1) -> (then, ReturnValue:Dlg Modal UI Object Reference)
   n7 = Get edt() -> (edt:Editor Utility Editable Text Box Object Reference)
   > Set Text(Text=Get Text(self=@n7), self:Text Box Object Reference=Get edt!(self:Dlg Modal UI Object Reference=@n6) -> (edt))
   > n8 = ShowModalEditorUtilityDialogInstance!(
    execute
    WidgetInstance:Editor Utility Dialog Widget Object Reference=@n6
    Title:Text="LOCTEXT(\"Instaneced Modal\")"
    ButtonLayout:EAppMsgType Enum=OkCancel
    Size:Vector 2D Structure
   ) -> (then, ReturnValue:Editor Utility Dialog Result Structure)
   n9 = Break Editor Utility Dialog Result(EditorUtilityDialogResult=@n8) -> (ReturnType=~)
   > ShowResults(Index=@n9, Object=@n9.WidgetInstance)
   > n10 = Cast To dlgModalUI!(execute, Object=@n9.WidgetInstance) -> (then, CastFailed, AsDlg Modal UI:Dlg Modal UI Object Reference) {is=title K2Node_DynamicCast}
   > Set Text(Text=Get Text(self=Get edt(self=@n10)), self=@n7)
   ShowResults(execute=@n5.then, Index=@n5.Result, Object=@n5.WidgetInstance)
   ShowEditorUtilityDialogInstance!(
    execute
    WidgetInstance:Editor Utility Dialog Widget Object Reference
    OnClosed:"Delegate (by ref)"
    Title:Text
    ButtonLayout:EAppMsgType Enum=OkCancel
    Size:Vector 2D Structure
   ) -> (then)
   Event PreConstruct()
   Event Construct()
   Event Tick()
  ]}
 ShowResults {nodes [
   n1 = ShowResults!() -> (then, Index:EAppReturnType Enum, Object:Object Reference) {is=title K2Node_FunctionEntry}
   > n2 = Cast To dlgModalUI(Object=@n1.Object)
   n3 = "To String (Float)"(InDouble=Get Value!(self:Slider Object Reference=Get slider!(self:Dlg Modal UI Object Reference=@n2) -> (slider:Editor Utility Slider Object Reference)) -> (Value:float))
   n4 = "To String (Text)"(InText=Select!(
     No:Text
     Yes:Text
     YesAll:Text
     NoAll:Text
     Cancel:Text=MakeLiteralText(Value="LOCTEXT(\"You cancelled !\")")
     Ok:Text=MakeLiteralText(Value="LOCTEXT(\"You pressed OK !\")")
     Retry:Text
     Continue:Text
     Index:EAppReturnType Enum=@n1
    ) -> (ReturnValue:Text))
   n5 = "To String (Text)"(InText=Get Text(self=Get edt(self=@n2)))
   > Show Message Dialog(
    Title=MakeLiteralText(Value="LOCTEXT(\"Result\")")
    Message="To Text (String)"(InString=Append!(A="Dialog Data:\r\n", B=@n4, C:String="\r\nTEXT:\r\n", D:String=@n5, E:String="\r\nVALUE:\r\n", F:String=@n3) -> (ReturnValue))
    MessageType=Ok
    DefaultValue=Ok
    MessageCategory=Success
   )
  ]}
]
variables [Dialog UI {type=Class, default=&/dlgModalUI.dlgModalUI_C}]
widgets {
 info {rootWidgetClass=&VerticalBox, widgetCount=8}
 widgets [
  VerticalBox_0 {widgetClassPath=&VerticalBox, widget=VerticalBox_0}
  EditorUtilityButton_0 {widgetClassPath=&EditorUtilityButton, parent=VerticalBox_0, widget=EditorUtilityButton_0, slot {class=VerticalBoxSlot}}
  TextBlock_50 {
   widgetClassPath=&TextBlock
   parent=EditorUtilityButton_0
   widget=TextBlock_50
   properties {text=Show Modal}
   slot {class=ButtonSlot}
  }
  EditorUtilityButton {widgetClassPath=&EditorUtilityButton, parent=VerticalBox_0, widget=EditorUtilityButton, slot {class=VerticalBoxSlot}}
  TextBlock {
   widgetClassPath=&TextBlock
   parent=EditorUtilityButton
   widget=TextBlock
   properties {text=Show Non Modal}
   slot {class=ButtonSlot}
  }
  edt {
   widgetClassPath=&EditorUtilityEditableTextBox
   bIsVariable=true
   parent=VerticalBox_0
   widget=edt
   properties {hintText=Input text}
   slot {class=VerticalBoxSlot, properties {padding=<ltrb 10 10 10 10>}}
  }
  btn_InstanceModal {widgetClassPath=&EditorUtilityButton, parent=VerticalBox_0, widget=btn_InstanceModal, slot {class=VerticalBoxSlot}}
  TextBlock_1 {
   widgetClassPath=&TextBlock
   parent=btn_InstanceModal
   widget=TextBlock_1
   properties {text=Show Instance Modal}
   slot {class=ButtonSlot}
  }
 ]
}
